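# Stereo MetaSound source from the Harmonix examples that plays a MIDI file through a Fusion sampler patch.
# Play, pause, continue, stop, kill and seek inputs become a transport that drives a MIDI Player with optional simple looping.
# The player's MIDI stream feeds the stereo Fusion Sampler, and its MIDI Clock is exposed as an output.
# A test bass MIDI file is the default, while Patch defaults to empty.
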MetaSoundSource /Harmonix/Examples/Sources/MSS_MIDIMusicPlayer
inputs [
 MIDI File {type=MIDIAsset}
 Patch {type=FusionPatchAsset}
 Simple Loop {type=Bool}
 UE.Source.On Continue {type=Trigger}
 UE.Source.On Kill {type=Trigger}
 UE.Source.On Pause {type=Trigger}
 UE.Source.On Stop {type=Trigger}
 UE.Source.OnPlay {type=Trigger}
 UE.Source.Seek {type=Trigger}
 UE.Source.Seek Target {type=MusicSeekTarget}
]
interfaces [UE.OutputFormat.Stereo:1.0, UE.Source:1.1]
outputs [
 MIDI Clock {type=MIDIClock}
 "UE.OutputFormat.Stereo.Audio:0" {type=Audio}
 "UE.OutputFormat.Stereo.Audio:1" {type=Audio}
]
pages [Default {
  input_defaults {
   MIDI File=/Harmonix/Tests/HarmonixMetasound/FusionSamplerNode/Midi/Bloom_BASS_Bb_92
   Patch=null
   Simple Loop=null
   UE.Source.On Continue=null
   UE.Source.On Kill=null
   UE.Source.On Pause=null
   UE.Source.On Stop=null
   UE.Source.OnPlay=null
   UE.Source.Seek=null
   UE.Source.Seek Target=null
  }
  nodes [
   HarmonixNodesMIDIPlayer_0 = HarmonixNodes.MIDIPlayer(
    MIDI File=^MIDI File
    Simple Loop=^Simple Loop
    Transport=HarmonixNodes.TriggerToTransport(
     Continue=^UE.Source.On Continue
     Kill=^UE.Source.On Kill
     Pause=^UE.Source.On Pause
     Stop=^UE.Source.On Stop
     Play=^UE.Source.OnPlay
     Trigger Seek=^UE.Source.Seek
     Seek Target=^UE.Source.Seek Target
    ).Transport
   ) {also {defaults {
      Kill Voices On MIDI File Change="<MetaSoundBuilderResult.SUCCEEDED: 0>"
      Kill Voices On Seek="<MetaSoundBuilderResult.SUCCEEDED: 0>"
      Pre-roll Bars=32
     }}}
   HarmonixNodesFusionSamplerStereo_0 = HarmonixNodes.FusionSamplerStereo(Patch=^Patch, MIDI Stream=@HarmonixNodesMIDIPlayer_0.MIDI Stream) {also {version [0,1]}}
  ]
  links [
   {dst="@UE.OutputFormat.Stereo.Audio:0", src=HarmonixNodesFusionSamplerStereo_0.Audio Left}
   {dst="@UE.OutputFormat.Stereo.Audio:1", src=HarmonixNodesFusionSamplerStereo_0.Audio Right}
   {dst="@MIDI Clock", src=HarmonixNodesMIDIPlayer_0.MIDI Clock}
  ]
 }]
properties {
 assetImportData=&self:AssetImportData
 bLooping=true
 duration=10000
 numChannels=2
 outputFormat=Stereo
 sampleRate=48000
}
