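# Harmonix example MetaSound source that plays a MIDI step sequence through a Fusion sampler patch in stereo.
# Source play, stop, pause, continue, kill and seek inputs become a transport driving a metronome and a step sequence player.
# The metronome supplies the MIDI clock from tempo and time signature; the player sends MIDI to the sampler.
# Sequence asset, patch, paging, step size and looping are exposed as inputs.
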MetaSoundSource /Harmonix/Examples/Sources/MSS_MIDISequencePlayer
inputs [
 "Metronome.Loop Length (Bars)" {type=Int32, access=Constructor}
 Metronome.Looping {type=Bool, access=Constructor}
 Patch {type=FusionPatchAsset}
 Sequence.Active Page {type=Float}
 Sequence.Auto Page {type=Bool}
 Sequence.Auto Page Plays Blank Pages {type=Bool}
 Sequence.Enabled {type=Bool}
 Sequence.Looping {type=Bool}
 Sequence.Step Size Quarter Notes {type=Float}
 Step Sequence Asset {type=MIDIStepSequenceAsset}
 Tempo {type=Float}
 Time Sig. Denominator {type=Int32}
 Time Sig. Numerator {type=Int32}
 UE.Source.On Continue {type=Trigger}
 UE.Source.On Kil {type=Trigger}
 UE.Source.On Pause {type=Trigger}
 UE.Source.On Seek {type=Trigger}
 UE.Source.On Stop {type=Trigger}
 UE.Source.OnPlay {type=Trigger}
 UE.Source.Seek Target {type=MusicSeekTarget}
]
interfaces [UE.OutputFormat.Stereo:1.0, UE.Source:1.1]
outputs [
 "UE.OutputFormat.Stereo.Audio:0" {type=Audio}
 "UE.OutputFormat.Stereo.Audio:1" {type=Audio}
]
pages [Default {
  input_defaults {
   "Metronome.Loop Length (Bars)"=4
   Metronome.Looping=null
   Patch=null
   Sequence.Active Page=0
   Sequence.Auto Page=null
   Sequence.Auto Page Plays Blank Pages="<MetaSoundBuilderResult.SUCCEEDED: 0>"
   Sequence.Enabled="<MetaSoundBuilderResult.SUCCEEDED: 0>"
   Sequence.Looping="<MetaSoundBuilderResult.SUCCEEDED: 0>"
   Sequence.Step Size Quarter Notes=0.25
   Step Sequence Asset=null
   Tempo=120
   Time Sig. Denominator=4
   Time Sig. Numerator=4
   UE.Source.On Continue=null
   UE.Source.On Kil=null
   UE.Source.On Pause=null
   UE.Source.On Seek=null
   UE.Source.On Stop=null
   UE.Source.OnPlay=null
   UE.Source.Seek Target=null
  }
  nodes [
   n1 = HarmonixNodes.TriggerToTransport(
    Continue=^UE.Source.On Continue
    Kill=^UE.Source.On Kil
    Pause=^UE.Source.On Pause
    Trigger Seek=^UE.Source.On Seek
    Stop=^UE.Source.On Stop
    Play=^UE.Source.OnPlay
    Seek Target=^UE.Source.Seek Target
   )
   HarmonixNodesFusionSamplerStereo_0 = HarmonixNodes.FusionSamplerStereo(
    Patch=^Patch
    MIDI Stream=HarmonixNodes.StepSequencePlayer(
     Active Page=^Sequence.Active Page
     Auto Page=^Sequence.Auto Page
     Auto Page Plays Blank Pages=^Sequence.Auto Page Plays Blank Pages
     Enable=^Sequence.Enabled
     Simple Loop=^Sequence.Looping
     Step Size Quarter Notes=^Sequence.Step Size Quarter Notes
     Step Sequence Asset=^Step Sequence Asset
     MIDI Clock=HarmonixNodes.Metronome(
      "Loop Length (Bars)"=^"Metronome.Loop Length (Bars)"
      Simple Loop=^Metronome.Looping
      Tempo=^Tempo
      Time Sig. Denominator=^Time Sig. Denominator
      Time Sig. Numerator=^Time Sig. Numerator
      Transport=@n1.Transport
     ) {also {defaults {Pre-roll Bars=32}}}.MIDI Clock
     Transport=@n1.Transport
    ) {also {version [0,1]}}.MIDI Stream
   ) {also {version [0,1]}}
  ]
  links [
   {dst="@UE.OutputFormat.Stereo.Audio:0", src=HarmonixNodesFusionSamplerStereo_0.Audio Left}
   {dst="@UE.OutputFormat.Stereo.Audio:1", src=HarmonixNodesFusionSamplerStereo_0.Audio Right}
  ]
 }]
properties {
 assetImportData=&self:AssetImportData
 bLooping=true
 duration=10000
 numChannels=2
 outputFormat=Stereo
}
